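# Physics asset for the Ms Alpine female medium body, covering dangling gear: three carabiners, a pocket, a dagger and a pistol.
# Each item's box body has tight swing and twist limits to a small capsule MSTR anchor, and carabiners A and B add a second link.
# Free-motion constraints pair carabiner and pocket anchors with spine_02 and the dagger anchor with the left thigh capsule.
# The pistol anchor has no constraint to the body.
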PhysicsAsset /BRCosmetics/Characters/Player/Female/Medium/Bodies/F_MED_Ms_Alpine/Meshes/F_MED_Ms_Alpine_Dyn
bodies {
 dyn_carabiner_a_01 [{shapeType=Box, extentX=4, extentY=1.5, extentZ=6, center=<1 0 0>}]
 dyn_carabiner_a_02 [{shapeType=Box, extentX=6, extentY=2.8, extentZ=0.8, center=<3 0 0>}]
 dyn_carabiner_a_MSTR [{shapeType=Capsule, length=1, radius=1}]
 dyn_carabiner_b_01 [{shapeType=Box, extentX=4, extentY=1.5, extentZ=6, center=<1 0 0>}]
 dyn_carabiner_b_02 [{shapeType=Box, extentX=6, extentY=2.8, extentZ=0.8, center=<3 0 0>}]
 dyn_carabiner_b_MSTR [{shapeType=Capsule, length=1, radius=1}]
 dyn_carabiner_c_01 [{shapeType=Box, extentX=4, extentY=1.5, extentZ=6, center=<1 0 0>}]
 dyn_carabiner_c_MSTR [{shapeType=Capsule, length=1, radius=1}]
 dyn_dagger [{shapeType=Box, extentX=25, extentY=7.5, extentZ=3, center=<2.5 0 1.5>, rotation=<rot -1 0 0>}]
 dyn_dagger_MSTR [{shapeType=Capsule, length=1, radius=1}]
 dyn_pistol [{shapeType=Box, extentX=13, extentY=9, extentZ=4.2, center=<1.7 0 2.3>}]
 dyn_pistol_MSTR [{shapeType=Capsule, length=1, radius=1}]
 dyn_pocket [{shapeType=Box, extentX=7.5, extentY=6.2, extentZ=3.6, center=<0.33 0 1.6>}]
 dyn_pocket_MSTR [{shapeType=Capsule, length=1, radius=1}]
 spine_01 [{shapeType=Capsule, length=4.1269, radius=7.4759, center=<0 0.6075 1.3645>}]
 spine_02 [{shapeType=Capsule, length=4.1269, radius=7.4759, center=<-1.351 1.3172 -0.1999>}]
 thigh_twist_01_l [{shapeType=Capsule, length=20, radius=8, center=<16.2 0 0.2>, rotation=<rot 95 0 0>}]
}
constraints [
 {
  bone1Name=dyn_carabiner_a_01
  bone2Name=dyn_carabiner_a_MSTR
  swing1LimitDegrees=15
  swing2LimitDegrees=45
  swing2Motion=Locked
  twistLimitDegrees=2
 }
 {bone1Name=dyn_carabiner_a_02, bone2Name=dyn_carabiner_a_01, swing1LimitDegrees=30, swing2LimitDegrees=30, twistLimitDegrees=5}
 {
  bone1Name=dyn_carabiner_b_01
  bone2Name=dyn_carabiner_b_MSTR
  swing1LimitDegrees=15
  swing2LimitDegrees=45
  swing2Motion=Locked
  twistLimitDegrees=2
 }
 {bone1Name=dyn_carabiner_b_02, bone2Name=dyn_carabiner_b_01, swing1LimitDegrees=30, swing2LimitDegrees=30, twistLimitDegrees=5}
 {
  bone1Name=dyn_carabiner_c_01
  bone2Name=dyn_carabiner_c_MSTR
  swing1LimitDegrees=15
  swing2LimitDegrees=45
  swing2Motion=Locked
  twistLimitDegrees=2
 }
 {bone1Name=dyn_pocket, bone2Name=dyn_pocket_MSTR, swing1LimitDegrees=10, swing2LimitDegrees=15, twistLimitDegrees=2}
 {bone1Name=dyn_dagger, bone2Name=dyn_dagger_MSTR, swing1LimitDegrees=10, swing2LimitDegrees=5, twistLimitDegrees=2}
 {bone1Name=dyn_pistol, bone2Name=dyn_pistol_MSTR, swing1LimitDegrees=5, swing2LimitDegrees=45, swing2Motion=Locked, twistLimitDegrees=2}
 {
  bone1Name=dyn_carabiner_a_MSTR
  bone2Name=spine_02
  swing1LimitDegrees=45
  swing1Motion=Free
  swing2LimitDegrees=45
  swing2Motion=Free
  twistLimitDegrees=45
  twistMotion=Free
 }
 {
  bone1Name=dyn_carabiner_b_MSTR
  bone2Name=spine_02
  swing1LimitDegrees=45
  swing1Motion=Free
  swing2LimitDegrees=45
  swing2Motion=Free
  twistLimitDegrees=45
  twistMotion=Free
 }
 {
  bone1Name=dyn_carabiner_c_MSTR
  bone2Name=spine_02
  swing1LimitDegrees=45
  swing1Motion=Free
  swing2LimitDegrees=45
  swing2Motion=Free
  twistLimitDegrees=45
  twistMotion=Free
 }
 {
  bone1Name=dyn_pocket_MSTR
  bone2Name=spine_02
  swing1LimitDegrees=45
  swing1Motion=Free
  swing2LimitDegrees=45
  swing2Motion=Free
  twistLimitDegrees=45
  twistMotion=Free
 }
 {
  bone1Name=spine_02
  bone2Name=spine_01
  swing1LimitDegrees=45
  swing1Motion=Free
  swing2LimitDegrees=45
  swing2Motion=Free
  twistLimitDegrees=45
  twistMotion=Free
 }
 {
  bone1Name=dyn_dagger_MSTR
  bone2Name=thigh_twist_01_l
  swing1LimitDegrees=45
  swing1Motion=Free
  swing2LimitDegrees=45
  swing2Motion=Free
  twistLimitDegrees=45
  twistMotion=Free
 }
]
