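# Physics asset for the dynamic collar and fur trim on the F_MED_Ms_Alpine female player body cosmetic.
# Five collar capsules (back, two left, two right) have narrow angular limits to small MSTR bodies, which link to spine_05 with free rotation.
# Ten small fur capsules hang off the back, l_02 and r_02 collar bodies with tight limits.
# Capsules on spine_05, neck_02 and head form a freely rotating torso-to-head chain.
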PhysicsAsset /BRCosmetics/Characters/Player/Female/Medium/Bodies/F_MED_Ms_Alpine/Meshes/F_MED_Ms_Alpine_Dyn_collar
bodies {
 dyn_collar_bk_01 [{shapeType=Capsule, length=8.5863, radius=3.5, center=<3 0 0>, rotation=<rot 0 0 90>}]
 dyn_collar_bk_01_MSTR [{shapeType=Capsule, length=1, radius=1}]
 dyn_collar_fur_bk_01 [{shapeType=Capsule, length=1.5, radius=0.5, center=<1.5 0 0>, rotation=<rot 90 0 0>}]
 dyn_collar_fur_bk_02 [{shapeType=Capsule, length=1.5, radius=0.5, center=<1.5 0 0>, rotation=<rot 90 0 0>}]
 dyn_collar_fur_bk_03 [{shapeType=Capsule, length=1.5, radius=0.5, center=<1.5 0 0>, rotation=<rot 90 0 0>}]
 dyn_collar_fur_l_01 [{shapeType=Capsule, length=1.5, radius=0.5, center=<1.5 0 0>, rotation=<rot 90 0 0>}]
 dyn_collar_fur_l_02 [{shapeType=Capsule, length=1.5, radius=0.5, center=<1.5 0 0>, rotation=<rot 90 0 0>}]
 dyn_collar_fur_l_03 [{shapeType=Capsule, length=1.5, radius=0.5, center=<1.5 0 0>, rotation=<rot 90 0 0>}]
 dyn_collar_fur_r_01 [{shapeType=Capsule, length=1.5, radius=0.5, center=<1.5 0 0>, rotation=<rot 90 0 0>}]
 dyn_collar_fur_r_02 [{shapeType=Capsule, length=1.5, radius=0.5, center=<1.5 0 0>, rotation=<rot 90 0 0>}]
 dyn_collar_fur_r_03 [{shapeType=Capsule, length=1.5, radius=0.5, center=<1.5 0 0>, rotation=<rot 90 0 0>}]
 dyn_collar_fur_r_04 [{shapeType=Capsule, length=1.5, radius=0.5, center=<1.5 0 0>, rotation=<rot 90 0 0>}]
 dyn_collar_l_01 [{shapeType=Capsule, length=3, radius=2.5, center=<1 0 0>, rotation=<rot 0 0 90>}]
 dyn_collar_l_01_MSTR [{shapeType=Capsule, length=1, radius=1}]
 dyn_collar_l_02 [{shapeType=Capsule, length=5, radius=3.2, center=<2 0 0>, rotation=<rot 0 0 90>}]
 dyn_collar_l_02_MSTR [{shapeType=Capsule, length=1, radius=1}]
 dyn_collar_r_01 [{shapeType=Capsule, length=3, radius=2.5, center=<1 0 0>, rotation=<rot 0 0 90>}]
 dyn_collar_r_01_MSTR [{shapeType=Capsule, length=1, radius=1}]
 dyn_collar_r_02 [{shapeType=Capsule, length=5, radius=3.2, center=<2 0 0>, rotation=<rot 0 0 90>}]
 dyn_collar_r_02_MSTR [{shapeType=Capsule, length=1, radius=1}]
 head [{shapeType=Capsule, length=2.8556, radius=7.7254, center=<3.2172 0.4378 0>, rotation=<rot 90 0 0>}]
 neck_02 [{shapeType=Capsule, length=3.9913, radius=3.861, rotation=<rot 90 0 0>}]
 spine_05 [{shapeType=Capsule, length=15.2159, radius=7.1192, center=<2.2082 -1.7287 -0.001>, rotation=<rot 0.043 -90 -0.2627>}]
}
constraints [
 {bone1Name=dyn_collar_bk_01, bone2Name=dyn_collar_bk_01_MSTR, swing1LimitDegrees=10, swing2LimitDegrees=20, twistLimitDegrees=5}
 {bone1Name=dyn_collar_l_02, bone2Name=dyn_collar_l_02_MSTR, swing1LimitDegrees=10, swing2LimitDegrees=20, twistLimitDegrees=5}
 {bone1Name=dyn_collar_l_01, bone2Name=dyn_collar_l_01_MSTR, swing1LimitDegrees=10, swing2LimitDegrees=20, twistLimitDegrees=5}
 {bone1Name=dyn_collar_r_02, bone2Name=dyn_collar_r_02_MSTR, swing1LimitDegrees=10, swing2LimitDegrees=20, twistLimitDegrees=5}
 {bone1Name=dyn_collar_r_01, bone2Name=dyn_collar_r_01_MSTR, swing1LimitDegrees=10, swing2LimitDegrees=20, twistLimitDegrees=5}
 {
  bone1Name=dyn_collar_bk_01_MSTR
  bone2Name=spine_05
  swing1LimitDegrees=45
  swing1Motion=Free
  swing2LimitDegrees=45
  swing2Motion=Free
  twistLimitDegrees=45
  twistMotion=Free
 }
 {
  bone1Name=dyn_collar_l_02_MSTR
  bone2Name=spine_05
  swing1LimitDegrees=45
  swing1Motion=Free
  swing2LimitDegrees=45
  swing2Motion=Free
  twistLimitDegrees=45
  twistMotion=Free
 }
 {
  bone1Name=dyn_collar_l_01_MSTR
  bone2Name=spine_05
  swing1LimitDegrees=45
  swing1Motion=Free
  swing2LimitDegrees=45
  swing2Motion=Free
  twistLimitDegrees=45
  twistMotion=Free
 }
 {
  bone1Name=dyn_collar_r_02_MSTR
  bone2Name=spine_05
  swing1LimitDegrees=45
  swing1Motion=Free
  swing2LimitDegrees=45
  swing2Motion=Free
  twistLimitDegrees=45
  twistMotion=Free
 }
 {
  bone1Name=dyn_collar_r_01_MSTR
  bone2Name=spine_05
  swing1LimitDegrees=45
  swing1Motion=Free
  swing2LimitDegrees=45
  swing2Motion=Free
  twistLimitDegrees=45
  twistMotion=Free
 }
 {
  bone1Name=neck_02
  bone2Name=spine_05
  swing1LimitDegrees=45
  swing1Motion=Free
  swing2LimitDegrees=45
  swing2Motion=Free
  twistLimitDegrees=45
  twistMotion=Free
 }
 {
  bone1Name=head
  bone2Name=neck_02
  swing1LimitDegrees=45
  swing1Motion=Free
  swing2LimitDegrees=45
  swing2Motion=Free
  twistLimitDegrees=45
  twistMotion=Free
 }
 {bone1Name=dyn_collar_fur_bk_01, bone2Name=dyn_collar_bk_01, swing1LimitDegrees=15, swing2LimitDegrees=15, twistLimitDegrees=5}
 {bone1Name=dyn_collar_fur_bk_02, bone2Name=dyn_collar_bk_01, swing1LimitDegrees=15, swing2LimitDegrees=15, twistLimitDegrees=5}
 {bone1Name=dyn_collar_fur_bk_03, bone2Name=dyn_collar_bk_01, swing1LimitDegrees=15, swing2LimitDegrees=15, twistLimitDegrees=5}
 {bone1Name=dyn_collar_fur_l_01, bone2Name=dyn_collar_l_02, swing1LimitDegrees=15, swing2LimitDegrees=15, twistLimitDegrees=5}
 {bone1Name=dyn_collar_fur_l_02, bone2Name=dyn_collar_l_02, swing1LimitDegrees=15, swing2LimitDegrees=15, twistLimitDegrees=5}
 {bone1Name=dyn_collar_fur_l_03, bone2Name=dyn_collar_l_02, swing1LimitDegrees=15, swing2LimitDegrees=15, twistLimitDegrees=5}
 {bone1Name=dyn_collar_fur_r_01, bone2Name=dyn_collar_r_02, swing1LimitDegrees=15, swing2LimitDegrees=15, twistLimitDegrees=5}
 {bone1Name=dyn_collar_fur_r_02, bone2Name=dyn_collar_r_02, swing1LimitDegrees=15, swing2LimitDegrees=15, twistLimitDegrees=5}
 {bone1Name=dyn_collar_fur_r_03, bone2Name=dyn_collar_r_02, swing1LimitDegrees=15, swing2LimitDegrees=15, twistLimitDegrees=5}
 {bone1Name=dyn_collar_fur_r_04, bone2Name=dyn_collar_r_02, swing1LimitDegrees=15, swing2LimitDegrees=15, twistLimitDegrees=5}
]
