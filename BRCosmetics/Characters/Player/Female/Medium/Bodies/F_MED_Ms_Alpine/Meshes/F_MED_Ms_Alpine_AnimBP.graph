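# Cosmetic AnimBP for the F_MED_Ms_Alpine female body that copies the attached parent mesh's pose and adds secondary motion.
# Thigh yaw relative to pelvis drives additive glute corrective offsets.
# Weighted constraints position carabiner, dagger, pistol and collar bones, with Look At nodes aiming dagger and pistol at helper bones.
# Two RigidBody passes simulate body and collar physics assets in base-bone space, at reduced alpha while crouching or targeting.
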AnimBlueprint /BRCosmetics/Characters/Player/Female/Medium/Bodies/F_MED_Ms_Alpine/Meshes/F_MED_Ms_Alpine_AnimBP
target_skeleton=&/F_MED_Ms_Alpine_Skeleton
graphs [
 AnimGraph {nodes [
   n1 = "Transform (Modify) Bone - Bone: dyn_glute_l"(
    ComponentPose=Local To Component(LocalPose=Copy Pose From Mesh() {bUseAttachedParent=true})
    Alpha=Get thigh_rot_up_l() -> (thigh_rot_up_l:double)
    Translation="-4.5,3.5,0.000"
    Rotation="0,0,0"
    Scale="1,1,1"
   ) {
    translationMode=BMM_Additive
    translationSpace=BCS_BoneSpace
    alphaBoolBlend {blendOption=Linear}
    boneToModify {boneName=dyn_glute_l}
    translation=<-4.5 3.5>
   }
   n2 = "Transform (Modify) Bone - Bone: dyn_glute_r"(ComponentPose=@n1, Alpha=Get thigh_rot_up_r() -> (thigh_rot_up_r:double), Translation="-4.5,3.5,0", Rotation="0,0,0", Scale="1,1,1") {
    translationMode=BMM_Additive
    translationSpace=BCS_BoneSpace
    alphaBoolBlend {blendOption=Linear}
    boneToModify {boneName=dyn_glute_r}
    translation=<-4.5 3.5>
   }
   n3 = "Constraint - dyn_carabiner_a_MSTR (PPPPP)"!(
    ComponentPose=@n2
    Alpha=1
    ConstraintWeights_0=0.267
    ConstraintWeights_1:float=0.304
    ConstraintWeights_2:float=0.361
    ConstraintWeights_3:float=0.057
    ConstraintWeights_4:float=0.011
   ) -> (Pose) {
    alphaBoolBlend {blendOption=Linear}
    boneToModify {boneName=dyn_carabiner_a_MSTR}
    constraintSetup [
     {targetBone {boneName=pelvis}}
     {targetBone {boneName=spine_01}}
     {targetBone {boneName=spine_02}}
     {targetBone {boneName=spine_03}}
     {targetBone {boneName=spine_04}}
    ]
    constraintWeights [1,1,1,1,1]
   }
   n4 = "Constraint - dyn_carabiner_b_MSTR (PPPPP)"!(
    ComponentPose=@n3
    Alpha=1
    ConstraintWeights_0=0.267
    ConstraintWeights_1:float=0.304
    ConstraintWeights_2:float=0.361
    ConstraintWeights_3:float=0.057
    ConstraintWeights_4:float=0.011
   ) -> (Pose) {
    alphaBoolBlend {blendOption=Linear}
    boneToModify {boneName=dyn_carabiner_b_MSTR}
    constraintSetup [
     {targetBone {boneName=pelvis}}
     {targetBone {boneName=spine_01}}
     {targetBone {boneName=spine_02}}
     {targetBone {boneName=spine_03}}
     {targetBone {boneName=spine_04}}
    ]
    constraintWeights [1,1,1,1,1]
   }
   n5 = "Constraint - dyn_carabiner_c_MSTR (PPPPP)"!(
    ComponentPose=@n4
    Alpha=1
    ConstraintWeights_0=0.267
    ConstraintWeights_1:float=0.304
    ConstraintWeights_2:float=0.361
    ConstraintWeights_3:float=0.057
    ConstraintWeights_4:float=0.011
   ) -> (Pose) {
    alphaBoolBlend {blendOption=Linear}
    boneToModify {boneName=dyn_carabiner_c_MSTR}
    constraintSetup [
     {targetBone {boneName=pelvis}}
     {targetBone {boneName=spine_01}}
     {targetBone {boneName=spine_02}}
     {targetBone {boneName=spine_03}}
     {targetBone {boneName=spine_04}}
    ]
    constraintWeights [1,1,1,1,1]
   }
   n6 = "Constraint - dyn_dagger_MSTR (PPP)"!(ComponentPose=@n5, Alpha=1, ConstraintWeights_0=0.094, ConstraintWeights_1:float=0.412, ConstraintWeights_2:float=0.494) -> (Pose) {
    alphaBoolBlend {blendOption=Linear}
    boneToModify {boneName=dyn_dagger_MSTR}
    constraintSetup [{targetBone {boneName=pelvis}}, {targetBone {boneName=thigh_l}}, {targetBone {boneName=thigh_twist_01_l}}]
    constraintWeights [1,1,1]
   }
   n7 = "Constraint - dyn_dagger_aim (PP)"!(ComponentPose=@n6, Alpha=1, ConstraintWeights_0=0.963, ConstraintWeights_1:float=0.037) -> (Pose) {
    alphaBoolBlend {blendOption=Linear}
    boneToModify {boneName=dyn_dagger_aim}
    constraintSetup [{targetBone {boneName=thigh_l}}, {targetBone {boneName=thigh_twist_01_l}}]
    constraintWeights [1,1]
   }
   n8 = Look At - Bone: dyn_dagger_MSTR(ComponentPose=@n7, Alpha=1) {
    alphaBoolBlend {blendOption=Linear}
    boneToModify {boneName=dyn_dagger_MSTR}
    lookAtLocation {x=0}
    lookAtTarget {boneReference {boneName=dyn_dagger_aim}}
    lookAt_Axis {axis=<1 0>}
   }
   n9 = "Constraint - dyn_pistol_MSTR (PP)"!(ComponentPose=@n8, Alpha=1, ConstraintWeights_0=0.929, ConstraintWeights_1:float=0.071) -> (Pose) {
    alphaBoolBlend {blendOption=Linear}
    boneToModify {boneName=dyn_pistol_MSTR}
    constraintSetup [{targetBone {boneName=pelvis}}, {targetBone {boneName=thigh_twist_01_r}}]
    constraintWeights [1,1]
   }
   n10 = "Constraint - dyn_pistol__aim (PPP)"!(ComponentPose=@n9, Alpha=1, ConstraintWeights_0=0.441, ConstraintWeights_1:float=0.502, ConstraintWeights_2:float=0.057) -> (Pose) {
    alphaBoolBlend {blendOption=Linear}
    boneToModify {boneName=dyn_pistol__aim}
    constraintSetup [{targetBone {boneName=pelvis}}, {targetBone {boneName=thigh_twist_01_r}}, {targetBone {boneName=thigh_r}}]
    constraintWeights [1,1,1]
   }
   n11 = Look At - Bone: dyn_pistol_MSTR(ComponentPose=@n10, Alpha=1) {
    alphaBoolBlend {blendOption=Linear}
    boneToModify {boneName=dyn_pistol_MSTR}
    lookAtLocation {x=0}
    lookAtTarget {boneReference {boneName=dyn_pistol__aim}}
    lookAt_Axis {axis=<1 0>}
   }
   n12 = "Constraint - dyn_collar_l_01_MSTR (PP)"!(ComponentPose=@n11, Alpha=1, ConstraintWeights_0=0.8, ConstraintWeights_1:float=0.2) -> (Pose) {
    alphaBoolBlend {blendOption=Linear}
    boneToModify {boneName=dyn_collar_l_01_MSTR}
    constraintSetup [{targetBone {boneName=spine_05}}, {targetBone {boneName=clavicle_l}}]
    constraintWeights [1,1]
   }
   n13 = "Constraint - dyn_collar_r_01_MSTR (PP)"!(ComponentPose=@n12, Alpha=1, ConstraintWeights_0=0.8, ConstraintWeights_1:float=0.2) -> (Pose) {
    alphaBoolBlend {blendOption=Linear}
    boneToModify {boneName=dyn_collar_r_01_MSTR}
    constraintSetup [{targetBone {boneName=spine_05}}, {targetBone {boneName=clavicle_r}}]
    constraintWeights [1,1]
   }
   n14 = "Constraint - dyn_collar_l_02_MSTR (PP)"!(ComponentPose=@n13, Alpha=1, ConstraintWeights_0=0.62, ConstraintWeights_1:float=0.38) -> (Pose) {
    alphaBoolBlend {blendOption=Linear}
    boneToModify {boneName=dyn_collar_l_02_MSTR}
    constraintSetup [{targetBone {boneName=spine_05}}, {targetBone {boneName=clavicle_l}}]
    constraintWeights [1,1]
   }
   n15 = "Constraint - dyn_collar_r_02_MSTR (PP)"!(ComponentPose=@n14, Alpha=1, ConstraintWeights_0=0.62, ConstraintWeights_1:float=0.38) -> (Pose) {
    alphaBoolBlend {blendOption=Linear}
    boneToModify {boneName=dyn_collar_r_02_MSTR}
    constraintSetup [{targetBone {boneName=spine_05}}, {targetBone {boneName=clavicle_r}}]
    constraintWeights [1,1]
   }
   n16 = Get rigid_alpha() -> (rigid_alpha:double)
   n17 = RigidBody(ComponentPose=@n15, Alpha=@n16) {
    bOverrideWorldGravity=true
    overridePhysicsAsset=&/F_MED_Ms_Alpine_Dyn
    simulationSpace=BaseBoneSpace
    alphaBoolBlend {blendOption=Linear}
    baseBoneRef {boneName=spine_01}
    componentAppliedLinearAccClamp=<3000 3000 3000>
    componentLinearAccScale=<0.35 0.35 0.35>
    componentLinearVelScale=<0.3 0.3 0.3>
    overrideWorldGravity {z=-980}
   }
   n18 = RigidBody(ComponentPose=@n17, Alpha=@n16) {
    bOverrideWorldGravity=true
    overridePhysicsAsset=&/F_MED_Ms_Alpine_Dyn_collar
    simulationSpace=BaseBoneSpace
    alphaBoolBlend {blendOption=Linear}
    baseBoneRef {boneName=spine_05}
    componentAppliedLinearAccClamp=<3000 3000 3000>
    componentLinearAccScale=<0.4 0.55 0.55>
    componentLinearVelScale=<0.4 0.4 0.4>
    overrideWorldGravity {z=-100}
   }
   Output Pose(Result=Component To Local(ComponentPose=@n18)) {layerGroup=None}
  ]}
 EventGraph {nodes [
   Event BlueprintUpdateAnimation()
   > n1 = Sequence()
   n2 = BreakTransform(InTransform=GetDeltaTransformFromRefPose(self=GetOwnerSkeletalMeshComponent(), BoneName=thigh_r, BaseName=pelvis))
   > Set thigh_rot_up_r(thigh_rot_up_r:double=MapRangeClamped(Value=BreakRotator(InRot=@n2.Rotation).Yaw, InRangeA=0, InRangeB=-100)) -> (Output_Get:double)
   n3 = BreakTransform(InTransform=GetDeltaTransformFromRefPose(self=GetOwnerSkeletalMeshComponent(), BoneName=thigh_l, BaseName=pelvis))
   > Set thigh_rot_up_l(thigh_rot_up_l:double=MapRangeClamped(Value=BreakRotator(InRot=@n3.Rotation).Yaw, InRangeA=-30, InRangeB=-100)) -> (Output_Get:double)
   n4 = Branch(execute=@n1.then_1, Condition=OR Boolean(A=Get bIsCrouching(), B=Get bIsTargeting()))
   > Set rigid_alpha(rigid_alpha:double=0.3) -> (Output_Get:double)
   Set rigid_alpha(execute=@n4.else, rigid_alpha=1)
  ]}
]
variables [
 rigid_alpha {type=number, default=1}
 thigh_rot_up_l {type=number}
 thigh_rot_up_r {type=number}
]
