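# Physics asset for the dynamic hair and hat of the Ms Alpine female player face accessory.
# Capsules cover the hair and hat bones, while head and spine_05 add larger head and upper-torso shapes, including angled boxes.
# Front hair, back hair, first main segment and hat hang off the head with tight angular limits.
# The lower main chain instead roots to spine_05 through free-rotating joints at a MSTR bone, then tightens down the chain.
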PhysicsAsset /BRCosmetics/Characters/Player/Female/Medium/Bodies/F_MED_Ms_Alpine/Meshes/Parts/F_MED_Ms_Alpine_FaceAcc_Dyn_hair
bodies {
 dyn_bk_hair_01 [{shapeType=Capsule, length=4, radius=2.5, center=<2 0 0>, rotation=<rot 90 0 0>}]
 dyn_ft_hair_01 [{shapeType=Capsule, length=2, radius=1, center=<0.5 0 0>, rotation=<rot 90 0 0>}]
 dyn_hair_main_01 [{shapeType=Capsule, length=4.5, radius=2.3, center=<3 0 0>, rotation=<rot 90 0 0>}]
 dyn_hair_main_02 [{shapeType=Capsule, length=3.8, radius=1.9, center=<2 0 0>, rotation=<rot 90 0 0>}]
 dyn_hair_main_02_MSTR [{shapeType=Capsule, length=1, radius=1}]
 dyn_hair_main_03 [{shapeType=Capsule, length=3.6, radius=1.67, center=<2 0 0>, rotation=<rot 90 0 0>}]
 dyn_hair_main_04 [
  {shapeType=Capsule, length=2.3033, radius=1.4, center=<2.1 0 0>, rotation=<rot 90 0 0>}
  {shapeType=Capsule, length=2.3033, radius=1.4, center=<2.1 0 -1.6>, rotation=<rot 90 0 0>}
 ]
 dyn_hat_01 [{shapeType=Capsule, length=4, radius=2.5, center=<1 0 0>, rotation=<rot 90 0 0>}]
 head [{shapeType=Capsule, length=5.0741, radius=6.8503, center=<3.3921 0.6556 0>, rotation=<rot 90 0 0>}]
 spine_05 [
  {shapeType=Capsule, length=23.173, radius=6.8318, center=<7.0399 -2.1638 0>}
  {shapeType=Capsule, length=16.7686, radius=4.1505, center=<11.9983 -0.4571 -3.4242>}
  {shapeType=Box, extentX=20.0582, extentY=25.4976, extentZ=1.9484, center=<5.8228 0.2997 -6.0674>, rotation=<rot -10 0 15.0009>}
  {shapeType=Box, extentX=20.0582, extentY=25.4976, extentZ=1.9484, center=<21.507 -1.2765 -14.5923>, rotation=<rot -45.8404 -14.3647 21.846>}
 ]
}
constraints [
 {bone1Name=dyn_ft_hair_01, bone2Name=head, swing1LimitDegrees=10, swing2LimitDegrees=5, twistLimitDegrees=2}
 {
  bone1Name=head
  bone2Name=spine_05
  swing1LimitDegrees=45
  swing1Motion=Free
  swing2LimitDegrees=45
  swing2Motion=Free
  twistLimitDegrees=45
  twistMotion=Free
 }
 {bone1Name=dyn_hat_01, bone2Name=head, swing1LimitDegrees=15, swing2LimitDegrees=20, twistLimitDegrees=5}
 {bone1Name=dyn_bk_hair_01, bone2Name=head, swing1LimitDegrees=20, swing2LimitDegrees=10, twistLimitDegrees=10}
 {bone1Name=dyn_hair_main_03, bone2Name=dyn_hair_main_02, swing1LimitDegrees=15, swing2LimitDegrees=20, twistLimitDegrees=5}
 {bone1Name=dyn_hair_main_04, bone2Name=dyn_hair_main_03, swing1LimitDegrees=15, swing2LimitDegrees=20, twistLimitDegrees=5}
 {bone1Name=dyn_hair_main_01, bone2Name=head, swing1LimitDegrees=15, swing2LimitDegrees=20, twistLimitDegrees=5}
 {
  bone1Name=dyn_hair_main_02_MSTR
  bone2Name=spine_05
  swing1LimitDegrees=45
  swing1Motion=Free
  swing2LimitDegrees=45
  swing2Motion=Free
  twistLimitDegrees=45
  twistMotion=Free
 }
 {
  bone1Name=dyn_hair_main_02
  bone2Name=dyn_hair_main_02_MSTR
  swing1LimitDegrees=45
  swing1Motion=Free
  swing2LimitDegrees=45
  swing2Motion=Free
  twistLimitDegrees=45
  twistMotion=Free
 }
]
